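# Material function that gives the Wild Cat character skin an animated blueprint look: light linework on blue paper.
# A time-stepped grid offset jumps the paper texture UVs, adding grid glints and distorting lines planar-projected from pre-skinned position.
# The body mask shapes linework and emissive and restores original areas.
# Emissive scales by time of day, and the Mask input blends the result with incoming attributes, keeping their normal.
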MaterialFunction /BRCosmetics/Characters/Player/Female/Medium/Bodies/F_MED_Wild_Cat/Skins/Blueprint/Materials/MF_WildCat
graph {nodes [
  n1 = GetMaterialAttributes(None=FunctionInput() {inputName=Material, inputType=FunctionInput_MaterialAttributes}) {attributeGetTypes [#1, #2, #3]}
  n2 = TextureCoordinate() {uTiling=2, vTiling=2}
  n3 = ScalarParameter() {defaultValue=2.5, group=AnimatedGrid, parameterName=PaperTiling}
  n4 = Constant() {r=0.5}
  n5 = ScalarParameter() {defaultValue=0.2, group=AnimatedGrid, parameterName=ShiftRange}
  n6 = Multiply(A=Time(), B=ScalarParameter() {defaultValue=0.2, group=AnimatedGrid, parameterName=AnimatedGridSpeed})
  n7 = Ceil(None=Divide(A=@n6) {constB=3})
  n8 = TextureSample(UVs=Add(A=Multiply(A=@n7) {constB=0.5}, B=TextureCoordinate())) {
   samplerType=SAMPLERTYPE_LinearGrayscale
   texture=/BRCosmetics/Characters/Player/Female/Medium/Bodies/F_MED_Wild_Cat/Textures/T_BlueYellowGridOffset
  }
  n9 = Frac(None=Add(A=@n8.R, B=@n6))
  n10 = StaticBool() {value=true}
  n11 = Constant() {r=3}
  n12 = Ceil(None=Multiply(
    A=MaterialFunctionCall!(ThresholdValue=@n4, ThresholdRange=@n5, Alpha=@n9, TwoSidedThreshold=@n10) {materialFunction=/Engine/Functions/Engine_MaterialFunctions02/Masking/ThresholdWithRange}.Result
    B=@n11
   ) {constB=2})
  n13 = TextureSample(UVs=Add(A=Multiply(A=@n2, B=@n3), B=Divide(A=@n12, B=@n11))) {texture=/BRCosmetics/Characters/Player/Female/Medium/Bodies/F_MED_Wild_Cat/Textures/T_BlueYellow_Paper}
  n14 = ScalarParameter() {defaultValue=16, group=AnimatedGrid, parameterName=PaperExponent}
  n15 = TextureSampleParameter2D() {
   group=Global
   parameterName=Mask
   texture=/BRCosmetics/Characters/Player/Female/Medium/Bodies/F_MED_Wild_Cat/Skins/Blueprint/Texture/T_F_Med_Wild_Cat_Blueprint_Body_Mask
  }
  n16 = TextureObject() {texture=/Game/Packages/DS_Fortnite_Mesh/BlueprintTest1}
  n17 = ScalarParameter() {defaultValue=8, group=Blueprint, parameterName=SizeOfLines}
  n18 = VertexInterpolator(VS=PreSkinnedPosition())
  n19 = MaterialFunctionCall(Float2=Rotator(Coordinate=ComponentMask(None=@n18.PS) {b=true, r=true}, Time=ScalarParameter() {group=Blueprint, parameterName=RotationAmt})) {materialFunction=&BreakOutFloat2Components}
  n20 = Add(
   A=Multiply(A=@n13.G, B=ScalarParameter() {defaultValue=0.5, group=Blueprint, parameterName=BlueprintDistortion})
   B=MaterialFunctionCall(X=@n19.R, Y=ComponentMask(None=@n18.PS) {g=true}, Z=@n19.R) {materialFunction=&MakeFloat3}.Result
  )
  n21 = VertexInterpolator(VS=PreSkinnedNormal())
  n22 = ScalarParameter() {parameterName=BlendContrast}
  n23 = Multiply(
   A=MaterialFunctionCall(In=OneMinus(None=@n15.R), Contrast=ScalarParameter() {defaultValue=1.9386, group=GmapMasking, parameterName=CheapContrast}) {materialFunction=&CheapContrast}.Result
   B=MaterialFunctionCall(Float3=MaterialFunctionCall(TextureObject=@n16, TextureSize=@n17, WorldPosition=@n20, World Space Normal=@n21.PS, ProjectionTransitionContrast=@n22) {materialFunction=&WorldAlignedTexture}.XY Texture) {materialFunction=&BreakOutFloat3Components}.R
  )
  n24 = Add(
   A=Multiply(
    A=Power(Base=@n13.B, Exp=@n14) {constExponent=8}
    B=ScalarParameter() {defaultValue=0.2, group=AnimatedGrid, parameterName=AnimatedGridIntensity}
   ) {constB=0.5}
   B=Multiply(A=Power(Base=@n13.R), B=Add(A=@n15.G, B=@n23))
  )
  n25 = FunctionInput(Preview=Constant4Vector().RGBA) {bUsePreviewValueAsDefault=true, inputName=Mask, inputType=FunctionInput_Vector4}
  n26 = ChannelMaskParameter(None=@n25) {maskChannel=Green, parameterName=MaskChannel, defaultValue=<rgba 0 1 0 0>}
  n27 = LinearInterpolate(
   A=VectorParameter() {group=Blueprint, parameterName=BlueColor, defaultValue=<rgba 0.0129 0.0896 0.29 1>}.RGB
   Alpha=Multiply(A=Clamp(None=@n24), B=OneMinus(None=@n26.R))
  )
  n28 = ScalarParameter() {defaultValue=1, group=Blueprint, parameterName=BlueprintIntensityBaseColor}
  n29 = Multiply(A=@n27, B=ScalarParameter() {defaultValue=1, group=Blueprint, parameterName=BlueprintIntensityEmissive})
  n30 = GetMaterialAttributes(None=SetMaterialAttributes!(MaterialAttributes=@n1.MaterialAttributes, BaseColor=Multiply(A=@n27, B=@n28), EmissiveColor=Multiply(A=@n29, B=@n15.R)) {attributeSetTypes [#2, #3]}) {attributeGetTypes [#2, #3]}
  n31 = LinearInterpolate(A=@n30.EmissiveColor, B=@n1.EmissiveColor, Alpha=@n15.B) {constB=0}
  n32 = ScalarParameter() {defaultValue=0.5, parameterName=EmissiveScaleMorning}
  n33 = ScalarParameter() {defaultValue=0.5, parameterName=EmissiveScaleEvening}
  n34 = ScalarParameter() {defaultValue=0.5, parameterName=EmissiveScaleNight}
  n35 = SetMaterialAttributes(
   MaterialAttributes=@n30.MaterialAttributes
   BaseColor=LinearInterpolate(A=@n30.BaseColor, B=@n1.BaseColor, Alpha=@n15.B)
   EmissiveColor=MaterialFunctionCall(Color_Morning=Multiply(A=@n31, B=@n32), Color_Day=@n31, Color_Evening=Multiply(A=@n31, B=@n33), Color_Night=Multiply(A=@n31, B=@n34)) {materialFunction=&Color_Time_of_Day}.Result
  ) {attributeSetTypes [#2, #3]}
  n36 = Add(A=ComponentMask(None=@n25) {r=true}, B=ComponentMask(None=@n25) {b=true})
  FunctionOutput(None=BlendMaterialAttributes(
    A=@n1.MaterialAttributes
    B=SetMaterialAttributes!(MaterialAttributes=@n35, Normal=@n1.Normal) {attributeSetTypes [#1]}
    Alpha=OneMinus(None=@n36)
   ) {vertexAttributeBlendType=UseA})
  MaterialFunctionCall!(Rotation Amount, PivotPoint, WorldPosition) {materialFunction=/Engine/Functions/Engine_MaterialFunctions02/WorldPositionOffset/RotateAboutWorldAxis_cheap}
  Reroute()
 ]}
