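# Master skeletal-mesh material for the Wild Cat medium female player body, built on material attributes and the shared character base functions.
# The PreFX function feeds an optional silk effect, masked by a selectable channel of the SkinFX_Mask texture.
# A second static switch optionally layers the Wild Cat-specific MF_WildCat effect using that same mask, before the PostFX function takes the result.
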Material /BRCosmetics/Characters/Player/Female/Medium/Bodies/F_MED_Wild_Cat/Materials/M_MED_WildCat_MASTER
graph {
 properties {
  bUseMaterialAttributes=true
  bUsedWithMeshDeformer=true
  bUsedWithMorphTargets=true
  bUsedWithSkeletalMesh=true
  pixelDepthOffsetMode=PDOM_Legacy
  usedShadingModels=Default Lit
 }
 nodes [
  n1 = MaterialFunctionCall() {materialFunction=&BaseFN_CharacterFunction_PreFX}
  n2 = TextureSampleParameter2D() {group=Skin FX, parameterName=SkinFX_Mask, texture=&T_BlacksRGB}
  n3 = StaticSwitchParameter(
   True=MaterialFunctionCall!(Alpha=ChannelMaskParameter(None=@n2.RGBA) {group=Skin FX, parameterName=Silk Channel}.R, Material=@n1.Result) {materialFunction=/Game/Weapons/WeaponSkins/MaterialFunction/MF_Silk}.Result
   False=@n1.Result
  ) {group=Skin FX, parameterName=Use Silk}
  MaterialFunctionCall(
   AO=@n1.AO
   Distance Blend=@n1.AO
   Fresnel=@n1.AO
   Gmap=@n1.AO
   Pre Skinned Local Position=@n1.AO
   Material=StaticSwitchParameter(
    True=MaterialFunctionCall!(Mask=@n2.RGBA, Material=@n3) {materialFunction=/BRCosmetics/Characters/Player/Female/Medium/Bodies/F_MED_Wild_Cat/Skins/Blueprint/Materials/MF_WildCat}.Result
    False=@n3
   ) {parameterName=UseWildCatTechArt}
  ) {materialFunction=&BaseFN_CharacterFunction_PostFX}
 ]
}
