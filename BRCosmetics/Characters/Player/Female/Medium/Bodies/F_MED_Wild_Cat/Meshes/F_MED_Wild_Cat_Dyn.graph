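# Physics asset for the Wild Cat female medium body cosmetic, covering secondary motion on its bag, two "drager" pieces and two zippers.
# Each dyn bone hangs off a small MSTR anchor body: zippers swing in a limited cone without twist, while bag and dragers swing narrowly on one axis.
# Capsules on neck and spine form the torso, linked by free constraints that also tie in the zipper anchors; bag and drager anchors have no torso constraint.
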PhysicsAsset /BRCosmetics/Characters/Player/Female/Medium/Bodies/F_MED_Wild_Cat/Meshes/F_MED_Wild_Cat_Dyn
bodies {
 dyn_bag [{shapeType=Capsule, length=8, radius=4.5, center=<1 0 0>, rotation=<rot 90 0 0>}]
 dyn_bag_MSTR [{shapeType=Capsule, length=0.5, radius=0.5}]
 dyn_drager_01 [{shapeType=Capsule, length=16, radius=2.6, center=<2 0 1>, rotation=<rot 90 0 0>}]
 dyn_drager_01_MSTR [{shapeType=Capsule, length=0.5, radius=0.5}]
 dyn_drager_02 [{shapeType=Capsule, length=16, radius=2.6, center=<2 0 1>, rotation=<rot 90 0 0>}]
 dyn_drager_02_MSTR [{shapeType=Capsule, length=0.5, radius=0.5}]
 dyn_zipper_01 [{shapeType=Capsule, length=0.5, radius=0.5, center=<1.5 0 0>, rotation=<rot 90 0 0>}]
 dyn_zipper_01_MSTR [{shapeType=Capsule, length=0.5, radius=0.5}]
 dyn_zipper_02 [{shapeType=Capsule, length=0.5, radius=0.5, center=<1.5 0 0>, rotation=<rot 90 0 0>}]
 dyn_zipper_02_MSTR [{shapeType=Capsule, length=0.5, radius=0.5}]
 neck_01 [{shapeType=Capsule, length=2.196, radius=4.7525, center=<-1.7411 0.812 0>, rotation=<rot 90 90 115.0041>}]
 spine_03 [{shapeType=Capsule, length=7.9807, radius=5.5125, center=<0 5.4349 0.0001>}]
 spine_05 [{shapeType=Capsule, length=9.027, radius=8.7072, center=<0.9445 -0.0159 0>}]
}
constraints [
 {
  bone1Name=spine_05
  bone2Name=spine_03
  swing1LimitDegrees=45
  swing1Motion=Free
  swing2LimitDegrees=45
  swing2Motion=Free
  twistLimitDegrees=45
  twistMotion=Free
 }
 {
  bone1Name=neck_01
  bone2Name=spine_05
  swing1LimitDegrees=45
  swing1Motion=Free
  swing2LimitDegrees=45
  swing2Motion=Free
  twistLimitDegrees=45
  twistMotion=Free
 }
 {
  bone1Name=dyn_zipper_01_MSTR
  bone2Name=neck_01
  swing1LimitDegrees=45
  swing1Motion=Free
  swing2LimitDegrees=45
  swing2Motion=Free
  twistLimitDegrees=45
  twistMotion=Free
 }
 {
  bone1Name=dyn_zipper_01
  bone2Name=dyn_zipper_01_MSTR
  swing1LimitDegrees=30
  swing2LimitDegrees=30
  twistLimitDegrees=45
  twistMotion=Locked
 }
 {
  bone1Name=dyn_zipper_02_MSTR
  bone2Name=spine_03
  swing1LimitDegrees=45
  swing1Motion=Free
  swing2LimitDegrees=45
  swing2Motion=Free
  twistLimitDegrees=45
  twistMotion=Free
 }
 {
  bone1Name=dyn_zipper_02
  bone2Name=dyn_zipper_02_MSTR
  swing1LimitDegrees=30
  swing2LimitDegrees=30
  twistLimitDegrees=45
  twistMotion=Locked
 }
 {
  bone1Name=dyn_drager_01
  bone2Name=dyn_drager_01_MSTR
  swing1LimitDegrees=10
  swing2LimitDegrees=15
  swing2Motion=Locked
  twistLimitDegrees=45
  twistMotion=Locked
 }
 {
  bone1Name=dyn_drager_02
  bone2Name=dyn_drager_02_MSTR
  swing1LimitDegrees=10
  swing2LimitDegrees=15
  swing2Motion=Locked
  twistLimitDegrees=45
  twistMotion=Locked
 }
 {
  bone1Name=dyn_bag
  bone2Name=dyn_bag_MSTR
  swing1LimitDegrees=10
  swing2LimitDegrees=45
  swing2Motion=Locked
  twistLimitDegrees=45
  twistMotion=Locked
 }
]
