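# Animation Blueprint for the F_MED_Wild_Cat female body cosmetic, copying the attached parent mesh's pose and adding secondary motion.
# Each update maps thigh and right-arm rotation from ref pose to alphas for additive corrective offsets on glute and right shoulder pad bones.
# Weighted constraints position drager, bag, zipper and shoulder pad master bones between body bones.
# A final pelvis-space RigidBody node simulates the Dyn physics asset with damped component motion.
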AnimBlueprint /BRCosmetics/Characters/Player/Female/Medium/Bodies/F_MED_Wild_Cat/Meshes/F_MED_Wild_Cat_AnimBP
target_skeleton=&/F_MED_Wild_Cat_Skeleton
graphs [
 AnimGraph {nodes [
   n1 = "Transform (Modify) Bone - Bone: dyn_glute_l"(
    ComponentPose=Local To Component(LocalPose=Copy Pose From Mesh() {bUseAttachedParent=true})
    Alpha=Get thigh_rot_up_l() -> (thigh_rot_up_l:double)
    Translation="-16,-2,0"
    Rotation="0,22,0"
    Scale="1,1,1"
   ) {
    rotationMode=BMM_Additive
    rotationSpace=BCS_BoneSpace
    translationMode=BMM_Additive
    translationSpace=BCS_BoneSpace
    alphaBoolBlend {blendOption=Linear}
    boneToModify {boneName=dyn_glute_l}
    rotation {yaw=22}
    translation=<-16 -2>
   }
   n2 = "Transform (Modify) Bone - Bone: dyn_glute_r"(ComponentPose=@n1, Alpha=Get thigh_rot_up_r() -> (thigh_rot_up_r:double), Translation="-16,-2,0", Rotation="0,22,0", Scale="1,1,1") {
    rotationMode=BMM_Additive
    rotationSpace=BCS_BoneSpace
    translationMode=BMM_Additive
    translationSpace=BCS_BoneSpace
    alphaBoolBlend {blendOption=Linear}
    boneToModify {boneName=dyn_glute_r}
    rotation {yaw=22}
    translation=<-16 -2>
   }
   n3 = "Constraint - dyn_drager_01_MSTR (PP)"!(ComponentPose=@n2, Alpha=1, ConstraintWeights_0=0.49, ConstraintWeights_1:float=0.51) -> (Pose) {
    alphaBoolBlend {blendOption=Linear}
    boneToModify {boneName=dyn_drager_01_MSTR}
    constraintSetup [{targetBone {boneName=thigh_r}}, {targetBone {boneName=thigh_twist_01_r}}]
    constraintWeights [1,1]
   }
   n4 = "Constraint - dyn_drager_02_MSTR (PP)"!(ComponentPose=@n3, Alpha=1, ConstraintWeights_0=0.47, ConstraintWeights_1:float=0.53) -> (Pose) {
    alphaBoolBlend {blendOption=Linear}
    boneToModify {boneName=dyn_drager_02_MSTR}
    constraintSetup [{targetBone {boneName=thigh_r}}, {targetBone {boneName=thigh_twist_01_r}}]
    constraintWeights [1,1]
   }
   n5 = "Constraint - dyn_bag_MSTR (PP)"!(ComponentPose=@n4, Alpha=1, ConstraintWeights_0=0.466, ConstraintWeights_1:float=0.534) -> (Pose) {
    alphaBoolBlend {blendOption=Linear}
    boneToModify {boneName=dyn_bag_MSTR}
    constraintSetup [{targetBone {boneName=thigh_l}}, {targetBone {boneName=thigh_twist_01_l}}]
    constraintWeights [1,1]
   }
   n6 = "Constraint - dyn_zipper_02_MSTR (PPPP)"!(
    ComponentPose=@n5
    Alpha=1
    ConstraintWeights_0=0.07
    ConstraintWeights_1:float=0.647
    ConstraintWeights_2:float=0.142
    ConstraintWeights_3:float=0.141
   ) -> (Pose) {
    alphaBoolBlend {blendOption=Linear}
    boneToModify {boneName=dyn_zipper_02_MSTR}
    constraintSetup [{targetBone {boneName=spine_02}}, {targetBone {boneName=spine_03}}, {targetBone {boneName=spine_04}}, {targetBone {boneName=spine_05}}]
    constraintWeights [1,1,1,1]
   }
   n7 = "Constraint - dyn_zipper_01_MSTR (PPPP)"!(
    ComponentPose=@n6
    Alpha=1
    ConstraintWeights_0=0.332
    ConstraintWeights_1:float=0.255
    ConstraintWeights_2:float=0.162
    ConstraintWeights_3:float=0.251
   ) -> (Pose) {
    alphaBoolBlend {blendOption=Linear}
    boneToModify {boneName=dyn_zipper_01_MSTR}
    constraintSetup [{targetBone {boneName=spine_05}}, {targetBone {boneName=neck_01}}, {targetBone {boneName=neck_02}}, {targetBone {boneName=head}}]
    constraintWeights [1,1,1,1]
   }
   n8 = "Constraint - dyn_shoulder_pad_r_MSTR (PP)"!(ComponentPose=@n7, Alpha=1, ConstraintWeights_0=0.3, ConstraintWeights_1:float=0.7) -> (Pose) {
    alphaBoolBlend {blendOption=Linear}
    boneToModify {boneName=dyn_shoulder_pad_r_MSTR}
    constraintSetup [{targetBone {boneName=upperarm_r}}, {targetBone {boneName=upperarm_twist_01_r}}]
    constraintWeights [1,1]
   }
   n9 = "Transform (Modify) Bone - Bone: dyn_shoulder_pad_r_MSTR"(ComponentPose=@n8, Alpha=Get arm_rot_up_r() -> (arm_rot_up_r:double), Translation="4.5,1.5,1", Rotation="-5,0,0", Scale="1,1,1") {
    rotationMode=BMM_Additive
    rotationSpace=BCS_BoneSpace
    translationMode=BMM_Additive
    translationSpace=BCS_BoneSpace
    alphaBoolBlend {blendOption=Linear}
    boneToModify {boneName=dyn_shoulder_pad_r_MSTR}
    rotation {pitch=-5}
    translation=<4.5 1.5 1>
   }
   n10 = "Transform (Modify) Bone - Bone: dyn_shoulder_pad_r_MSTR"(ComponentPose=@n9, Alpha=Get arm_rot_dn_r() -> (arm_rot_dn_r:double), Translation="1,0,0.8", Rotation="0,0,0", Scale="1,1,1") {
    rotationMode=BMM_Additive
    rotationSpace=BCS_BoneSpace
    translationMode=BMM_Additive
    translationSpace=BCS_BoneSpace
    alphaBoolBlend {blendOption=Linear}
    boneToModify {boneName=dyn_shoulder_pad_r_MSTR}
    translation {x=1, z=0.8}
   }
   n11 = RigidBody(ComponentPose=@n10) {
    bClampLinearTranslationLimitToRefPose=true
    bOverrideWorldGravity=true
    overridePhysicsAsset=&/F_MED_Wild_Cat_Dyn
    simulationSpace=BaseBoneSpace
    alphaBoolBlend {blendOption=Linear}
    baseBoneRef {boneName=pelvis}
    componentAppliedLinearAccClamp=<800 800 800>
    componentLinearAccScale=<0.15 0.15 0.15>
    componentLinearVelScale=<0.1 0.1 0.1>
    overrideWorldGravity {z=-980}
   }
   Output Pose(Result=Component To Local(ComponentPose=@n11)) {layerGroup=None}
  ]}
 EventGraph {nodes [
   Event BlueprintUpdateAnimation()
   n1 = BreakTransform(InTransform=GetDeltaTransformFromRefPose(self=GetOwnerSkeletalMeshComponent(), BoneName=thigh_r, BaseName=pelvis))
   > Set thigh_rot_up_r(thigh_rot_up_r:double=MapRangeClamped(Value=BreakRotator(InRot=@n1.Rotation).Yaw, InRangeA=0, InRangeB=-100)) -> (Output_Get:double)
   n2 = BreakTransform(InTransform=GetDeltaTransformFromRefPose(self=GetOwnerSkeletalMeshComponent(), BoneName=thigh_l, BaseName=pelvis))
   > Set thigh_rot_up_l(thigh_rot_up_l:double=MapRangeClamped(Value=BreakRotator(InRot=@n2.Rotation).Yaw, InRangeA=-30, InRangeB=-100)) -> (Output_Get:double)
   n3 = BreakTransform(InTransform=GetDeltaTransformFromRefPose(self=GetOwnerSkeletalMeshComponent(), BoneName=upperarm_r, BaseName=spine_05))
   > Set arm_rot_up_r(arm_rot_up_r:double=MapRangeClamped(Value=BreakRotator(InRot=@n3.Rotation).Pitch, InRangeA=0, InRangeB=70)) -> (Output_Get:double)
   n4 = BreakTransform(InTransform=GetDeltaTransformFromRefPose(self=GetOwnerSkeletalMeshComponent(), BoneName=upperarm_twist_01_r, BaseName=spine_05))
   > Set arm_rot_dn_r(arm_rot_dn_r:double=MapRangeClamped(Value=BreakRotator(InRot=@n4.Rotation).Pitch, InRangeA=0, InRangeB=-30)) -> (Output_Get:double)
  ]}
]
variables [
 arm_rot_dn_r {type=number}
 arm_rot_up_r {type=number}
 thigh_rot_up_l {type=number}
 thigh_rot_up_r {type=number}
]
