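# Physics asset for the hair and hat dynamics of the F_MED_Wild_Cat female player cosmetic, built from one capsule per bone.
# Left and right strands form two- or three-link chains from the head, while four back-hair bones branch in parallel from a base body on the head.
# The hat also joins the head, and hair and hat joints carry narrow swing and twist limits.
# Larger head and neck capsules share a free-rotating constraint.
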PhysicsAsset /BRCosmetics/Characters/Player/Female/Medium/Bodies/F_MED_Wild_Cat/Meshes/Parts/F_MED_Wild_Cat_Hair_Dyn
bodies {
 dyn_hair_bk_01 [{shapeType=Capsule, length=1, radius=1, center=<1 0 0>, rotation=<rot 90 0 0>}]
 dyn_hair_bk_02 [{shapeType=Capsule, length=1, radius=1, center=<1 0 0>, rotation=<rot 90 0 0>}]
 dyn_hair_bk_03 [{shapeType=Capsule, length=1, radius=1, center=<1 0 0>, rotation=<rot 90 0 0>}]
 dyn_hair_bk_04 [{shapeType=Capsule, length=1, radius=1, center=<1 0 0>, rotation=<rot 90 0 0>}]
 dyn_hair_bk_base [{shapeType=Capsule, length=2.914, radius=2.086, center=<3 0 0>, rotation=<rot 0 0 90>}]
 dyn_hair_l_a_01 [{shapeType=Capsule, length=1, radius=1, center=<1 0 0>, rotation=<rot 90 0 0>}]
 dyn_hair_l_a_02 [{shapeType=Capsule, length=1, radius=1, center=<1 0 0>, rotation=<rot 90 0 0>}]
 dyn_hair_l_b_01 [{shapeType=Capsule, length=1, radius=1, center=<1 0 0>, rotation=<rot 90 0 0>}]
 dyn_hair_l_b_02 [{shapeType=Capsule, length=1, radius=1, center=<1 0 0>, rotation=<rot 90 1.6876 1.6879>}]
 dyn_hair_l_b_03 [{shapeType=Capsule, length=1, radius=1, center=<1 0 0>, rotation=<rot 90 5.0511 5.0507>}]
 dyn_hair_l_c_01 [{shapeType=Capsule, length=1, radius=1, center=<1 0 0>, rotation=<rot 90 0 0>}]
 dyn_hair_l_c_02 [{shapeType=Capsule, length=1, radius=1, center=<1 0 0>, rotation=<rot 90 0 0>}]
 dyn_hair_r_a_01 [{shapeType=Capsule, length=1, radius=1, center=<1 0 0>, rotation=<rot 90 0 0>}]
 dyn_hair_r_a_02 [{shapeType=Capsule, length=1, radius=1, center=<1 0 0>, rotation=<rot 90 0 0>}]
 dyn_hair_r_b_01 [{shapeType=Capsule, length=1, radius=1, center=<1 0 0>, rotation=<rot 90 0 0>}]
 dyn_hair_r_b_02 [{shapeType=Capsule, length=1, radius=1, center=<1 0 0>, rotation=<rot 90 0 0>}]
 dyn_hair_r_b_03 [{shapeType=Capsule, length=1, radius=1, center=<1 0 0>, rotation=<rot 90 0 0>}]
 dyn_hair_r_c_01 [{shapeType=Capsule, length=1, radius=1, center=<1 0 0>, rotation=<rot 90 0 0>}]
 dyn_hair_r_c_02 [{shapeType=Capsule, length=1, radius=1, center=<1 0 0>, rotation=<rot 90 0 0>}]
 dyn_hat [{shapeType=Capsule, length=6, radius=2.5, center=<1 0 0>, rotation=<rot 90 0 0>}]
 head [{shapeType=Capsule, length=5.1949, radius=5.9727, center=<1.8699 2.8472 0>, rotation=<rot 90 0 0>}]
 neck_01 [{shapeType=Capsule, length=5.1145, radius=3.9963, center=<4.2395 0.5635 -0.0002>, rotation=<rot 90 0 0>}]
}
constraints [
 {bone1Name=dyn_hair_bk_base, bone2Name=head, swing1LimitDegrees=5, swing2LimitDegrees=45, twistLimitDegrees=5}
 {bone1Name=dyn_hair_bk_01, bone2Name=dyn_hair_bk_base, swing1LimitDegrees=20, swing2LimitDegrees=20, twistLimitDegrees=5}
 {bone1Name=dyn_hair_bk_02, bone2Name=dyn_hair_bk_base, swing1LimitDegrees=20, swing2LimitDegrees=20, twistLimitDegrees=5}
 {bone1Name=dyn_hair_bk_03, bone2Name=dyn_hair_bk_base, swing1LimitDegrees=20, swing2LimitDegrees=20, twistLimitDegrees=5}
 {bone1Name=dyn_hair_bk_04, bone2Name=dyn_hair_bk_base, swing1LimitDegrees=20, swing2LimitDegrees=20, twistLimitDegrees=5}
 {bone1Name=dyn_hair_l_a_01, bone2Name=head, swing1LimitDegrees=15, swing2LimitDegrees=20, twistLimitDegrees=5}
 {bone1Name=dyn_hair_l_a_02, bone2Name=dyn_hair_l_a_01, swing1LimitDegrees=15, swing2LimitDegrees=20, twistLimitDegrees=5}
 {bone1Name=dyn_hair_l_b_01, bone2Name=head, swing1LimitDegrees=15, swing2LimitDegrees=20, twistLimitDegrees=5}
 {bone1Name=dyn_hair_l_b_02, bone2Name=dyn_hair_l_b_01, swing1LimitDegrees=15, swing2LimitDegrees=20, twistLimitDegrees=5}
 {bone1Name=dyn_hair_l_b_03, bone2Name=dyn_hair_l_b_02, swing1LimitDegrees=15, swing2LimitDegrees=20, twistLimitDegrees=5}
 {bone1Name=dyn_hair_l_c_01, bone2Name=head, swing1LimitDegrees=15, swing2LimitDegrees=20, twistLimitDegrees=5}
 {bone1Name=dyn_hair_l_c_02, bone2Name=dyn_hair_l_c_01, swing1LimitDegrees=15, swing2LimitDegrees=20, twistLimitDegrees=5}
 {bone1Name=dyn_hair_r_a_01, bone2Name=head, swing1LimitDegrees=15, swing2LimitDegrees=20, twistLimitDegrees=5}
 {bone1Name=dyn_hair_r_a_02, bone2Name=dyn_hair_r_a_01, swing1LimitDegrees=15, swing2LimitDegrees=20, twistLimitDegrees=5}
 {bone1Name=dyn_hair_r_b_01, bone2Name=head, swing1LimitDegrees=15, swing2LimitDegrees=20, twistLimitDegrees=5}
 {bone1Name=dyn_hair_r_b_02, bone2Name=dyn_hair_r_b_01, swing1LimitDegrees=15, swing2LimitDegrees=20, twistLimitDegrees=5}
 {bone1Name=dyn_hair_r_b_03, bone2Name=dyn_hair_r_b_02, swing1LimitDegrees=15, swing2LimitDegrees=20, twistLimitDegrees=5}
 {bone1Name=dyn_hair_r_c_01, bone2Name=head, swing1LimitDegrees=15, swing2LimitDegrees=20, twistLimitDegrees=5}
 {bone1Name=dyn_hair_r_c_02, bone2Name=dyn_hair_r_c_01, swing1LimitDegrees=15, swing2LimitDegrees=20, twistLimitDegrees=5}
 {bone1Name=dyn_hat, bone2Name=head, swing1LimitDegrees=15, swing2LimitDegrees=10, twistLimitDegrees=5}
 {
  bone1Name=head
  bone2Name=neck_01
  swing1LimitDegrees=45
  swing1Motion=Free
  swing2LimitDegrees=45
  swing2Motion=Free
  twistLimitDegrees=45
  twistMotion=Free
 }
]
